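# Translucent vehicle cosmetic material, named as the Coronado IAD loading effect, drawing a glitchy holographic scan over the scene.
# Screen-aligned digital noise, rotated in steps by quantized cloud noise, pans to drive color variation and UV distortion.
# A time-driven band sweeps along local X, boosting glow and distortion.
# A four-tap SceneDepth edge detect forms the outline, combined in emissive with distorted, exposure-compensated SceneColor.
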Material /VehicleCosmetics/IAD/M_Coronado_IAD_Loading
graph {
 properties {bAutomaticallySetUsageInEditor=false, blendMode=BLEND_Translucent, translucencyPass=MTP_BeforeDOF, usedShadingModels=Default Lit}
 nodes [
  n1 = VectorParameter() {defaultValue=<rgba 0.156 0.245 1 1>}
  n2 = MaterialFunctionCall!(Center Point World Position=ObjectPositionWS()) {materialFunction=/Engine/Functions/Engine_MaterialFunctions02/UVs/WorldSpaceAlignedScreenCoordinates}
  n3 = Multiply(A=@n2.Offset and Scaled UVS, B=ScalarParameter() {defaultValue=4, parameterName=Param_4})
  n4 = ComponentMask(None=@n2.Offset and Scaled UVS) {r=true}
  n5 = ScalarParameter() {defaultValue=0.001, parameterName=Param_5}
  n6 = TextureSample(UVs=Panner(Coordinate=Multiply(A=@n4, B=@n5)) {speedX=0.02, speedY=1}) {texture=/Game/Global/Textures/Default/Noise/T_Noise_Cloud}
  n7 = Multiply(A=Round(None=Multiply(A=@n6.B) {constB=4})) {constB=0.25}
  n8 = TextureSample(UVs=Panner(Coordinate=MaterialFunctionCall(UVs=@n3, "Rotation Angle (0-1)"=@n7) {materialFunction=&CustomRotator}.Rotated Values) {speedY=0.2}) {samplerType=SAMPLERTYPE_LinearColor, texture=/Game/Global/Textures/Default/Noise/T_Atlas_DigitalNoise_01}
  n9 = ComponentMask(None=@n8.RGB) {r=true}
  n10 = LinearInterpolate(
   A=VectorParameter() {parameterName=Param_1, defaultValue=<rgba 0.0549 0.3399 0.526 1>}.RGB
   B=VectorParameter() {parameterName=Param_2, defaultValue=<rgba 0.1979 0.0595 0.0056 1>}.RGB
   Alpha=Multiply(A=@n9, B=@n9)
  )
  n11 = ComponentMask(None=@n8.RGB) {b=true}
  n12 = Multiply(A=@n11, B=@n11)
  n13 = Multiply(A=ScalarParameter() {defaultValue=800, parameterName=Param_8}, B=Time())
  n14 = ComponentMask(None=LocalPosition().XYZ) {r=true}
  n15 = Frac(None=Multiply(A=Add(A=@n13, B=@n14)) {constB=0.001})
  n16 = Multiply(A=@n15, B=@n15)
  n17 = Multiply(A=@n16, B=@n16)
  n18 = Multiply(A=@n17, B=@n17)
  n19 = Sine(None=Multiply(A=@n18, B=@n18))
  n20 = Multiply(A=@n19, B=ScalarParameter() {defaultValue=0.03, parameterName=Distortion})
  n21 = Add(A=@n20) {constB=0.01}
  n22 = Add(
   A=Multiply(A=LinearInterpolate(A=@n1.RGB, B=@n10, Alpha=@n12), B=NamedRerouteUsage())
   B=Multiply(A=Saturate(None=@n21), B=Multiply(A=@n1.RGB) {constB=5})
  )
  n23 = Multiply(A=SceneColor!(UVs=NamedRerouteUsage()).RGB, B=EyeAdaptation().EyeAdaptation)
  MaterialFunctionCall(Emissive=LinearInterpolate(A=@n22, B=Add(A=@n22, B=@n23), Alpha=ScalarParameter() {defaultValue=0.6, parameterName=Param_1})) {materialFunction=&MF_FX_EmissiveAdjustments}
  n24 = LinearInterpolate(A=NamedRerouteUsage(), B=ScreenPosition().ViewportUV)
  n25 = Constant2Vector() {r=1}
  n26 = Multiply(A=SceneTexelSize(), B=ScalarParameter() {defaultValue=1, parameterName=Outline Thickness})
  n27 = SceneDepth(UVs=Add(A=@n24, B=Multiply(A=@n25, B=@n26)))
  n28 = Constant2Vector() {r=-1}
  n29 = SceneDepth(UVs=Add(A=@n24, B=Multiply(A=@n28, B=@n26)))
  n30 = Constant2Vector() {g=1}
  n31 = SceneDepth(UVs=Add(A=@n24, B=Multiply(A=@n30, B=@n26)))
  n32 = Constant2Vector() {g=-1}
  n33 = SceneDepth(UVs=Add(A=@n24, B=Multiply(A=@n32, B=@n26)))
  n34 = Add(
   A=Add(A=ComponentMask(None=@n27.R) {r=true}, B=ComponentMask(None=@n29.R) {r=true})
   B=Add(A=ComponentMask(None=@n31.R) {r=true}, B=ComponentMask(None=@n33.R) {r=true})
  )
  n35 = Multiply(A=ComponentMask(None=SceneDepth(UVs=@n24).R) {r=true}) {constB=4}
  n36 = Power(Base=Multiply(A=@n8.G, B=ConstantBiasScale(None=@n19)), Exp=ScalarParameter() {defaultValue=0.3, parameterName=Param_6})
  n37 = Power(
   Base=Divide(A=Subtract(A=@n34, B=@n35), B=ScalarParameter() {defaultValue=5, parameterName=Outline Div})
   Exp=LinearInterpolate(A=ScalarParameter() {defaultValue=8, parameterName=Param_7}, B=ScalarParameter() {parameterName=OutlineExp}, Alpha=Saturate(None=@n36)) {constA=1.5}
  )
  NamedRerouteDeclaration(None=Multiply(A=Saturate(None=@n37), B=ScalarParameter() {defaultValue=5, parameterName=Outline Intensity})) {name=Outline, nodeColor=<rgba 1 0 0.4922 1>}
  n38 = LinearInterpolate(A=MaterialFunctionCall() {materialFunction=&DitherTemporalAA}.Result) {constAlpha=1}
  n39 = Multiply(A=Subtract(A=ComponentMask(None=@n8.RGB) {g=true, r=true}) {constB=0.5}, B=Add(A=@n20) {constB=0.002})
  NamedRerouteDeclaration(None=Subtract(A=ScreenPosition().ViewportUV, B=Multiply(A=@n38, B=@n39) {constA=1})) {name=Distorted UVs, nodeColor=<rgba 0 0.8125 1 1>}
  ScalarParameter() {defaultValue=1, parameterName=Opacity}
 ]
}
